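# Translucent, lit master material for bubble particles on sprites and meshes, with index-of-refraction refraction.
# Sphere masks over noise-wobbled UVs build center fill, edge and center highlights procedurally, plus a panning-noise dissolve.
# Dynamic parameters supply dissolve, wiggle, variation, refraction and optional colors, alongside particle color, time-of-day tinting and scene-color blending.
# Opacity has optional depth fade, switched by shading path and feature level; a view-space reflection lookup fakes a cubemap.
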Material /DealershipSharedLibrary/FX/Materials/MasterMaterials/Liquid/Bubble/M_DLR_Bubbles
graph {
 properties {
  bCastRayTracedShadows=false
  bUsedWithMeshParticles=true
  bUsedWithNiagaraSprites=true
  bUsedWithParticleSprites=true
  blendMode=BLEND_Translucent
  pixelDepthOffsetMode=PDOM_Legacy
  refractionMethod=RM_IndexOfRefraction
  usedShadingModels=Default Lit
 }
 nodes [
  Clamp() {maxDefault=0.5, minDefault=-0.5}
  MaterialFunctionCall(Emissive=NamedRerouteUsage()) {materialFunction=&MF_FX_EmissiveAdjustments}
  n1 = NamedRerouteUsage()
  Multiply(A=@n1) {constB=0.25}
  NamedRerouteDeclaration(None=Constant2Vector() {g=0.5, r=0.5}) {name=Bubble Center, nodeColor=<rgba 1 0.8541 0 1>}
  n2 = DynamicParameter() {defaultValue=<rgba 0 0.3 0 1>, paramNames [Dissolve, Wiggle, Highlight Variation, Refraction]}
  NamedRerouteDeclaration(None=@n2.Highlight Variation) {name=Variation, nodeColor=<rgba 1 0 0.0132 1>}
  n3 = TextureSampleParameter2D(UVs=TextureCoordinate() {uTiling=0.005, vTiling=0.005}) {
   parameterName=Distortion Texture
   samplerType=SAMPLERTYPE_Normal
   texture=&T_Weap_Distort_N_02
  }
  n4 = Transform(None=ReflectionVectorWS(CustomWorldNormal=@n3.RGB)) {transformSourceType=TRANSFORMSOURCE_World, transformType=TRANSFORM_View}
  n5 = VectorParameter() {parameterName=CubeMapOffset, defaultValue=<rgba 2 2 0 1>}
  n6 = TextureSampleParameter2D(UVs=Multiply(A=ComponentMask(None=@n4) {g=true, r=true}, B=ComponentMask(None=@n5.RGB) {g=true, r=true})) {
   parameterName=SphereReflectionMap
   samplerType=SAMPLERTYPE_Masks
   texture=/Game/Characters/Player/Male/Large/Bodies/M_LRG_Clash_V/Textures/Simple_Perlin_M
  }
  NamedRerouteDeclaration(None=Multiply(
    A=Power(Base=@n6.RGB) {constExponent=4}
    B=ScalarParameter() {defaultValue=0.75, group=reflection, parameterName=Cubemap Reflection - Amount}
   )) {name=Cubemap, nodeColor=<rgba 1 0.3864 0.3813 1>}
  n7 = NamedRerouteUsage()
  n8 = TextureCoordinate() {uTiling=0.4, vTiling=0.4}
  n9 = Multiply(A=NamedRerouteUsage()) {constB=5}
  n10 = TextureSampleParameter2D(UVs=Panner(Coordinate=Add(A=@n8, B=@n9)) {speedY=0.2}) {
   parameterName=Distortion Texture
   samplerType=SAMPLERTYPE_Normal
   texture=&T_Weap_Distort_N_02
  }
  n11 = Multiply(A=ComponentMask(None=@n10.RGB) {g=true, r=true}) {constB=0.4}
  n12 = NamedRerouteUsage()
  n13 = SphereMask(A=Add(A=Add(A=@n7, B=@n11), B=Multiply(A=@n12) {constB=3}), B=NamedRerouteUsage(), Radius=NamedRerouteUsage()) {attenuationRadius=0.4}
  NamedRerouteDeclaration(None=Clamp(None=OneMinus(None=@n13))) {name=Dissolve_output, nodeColor=<rgba 1 0.3916 0.8308 1>}
  NamedRerouteDeclaration(None=LinearInterpolate(B=NamedRerouteUsage(), Alpha=NamedRerouteUsage()) {constA=1, constB=1.125}) {name=Refraction, nodeColor=<rgba 0.1719 0.1719 0.1719 1>}
  n14 = TextureCoordinate() {uTiling=0.2, vTiling=0.2}
  n15 = Multiply(A=NamedRerouteUsage()) {constB=5}
  n16 = TextureSampleParameter2D(UVs=Panner(Coordinate=Add(A=@n14, B=@n15)) {speedY=0.6}) {
   parameterName=Distortion Texture
   samplerType=SAMPLERTYPE_Normal
   texture=&T_Weap_Distort_N_02
  }
  n17 = ComponentMask(None=@n16.RGB) {g=true, r=true}
  n18 = NamedRerouteUsage()
  NamedRerouteDeclaration(None=Add(A=TextureCoordinate(), B=Multiply(A=@n17, B=@n18))) {name=UV Offset, nodeColor=<rgba 0.8796 1 0.2542 1>}
  n19 = DynamicParameter() {parameterIndex=1, paramNames [Base Color R, Base Color G, Base Color G, Base Color A]}
  NamedRerouteDeclaration(None=MaterialFunctionCall(R=@n19.Base Color R, G=@n19.Base Color R, B=@n19.Base Color R) {materialFunction=&AppendMany}.RGB) {name=Base Color, nodeColor=<rgba 0.2656 1 0 1>}
  NamedRerouteDeclaration(None=@n19.Base Color A) {name=Base Color A, nodeColor=<rgba 0.2656 1 0 1>}
  n20 = DynamicParameter() {parameterIndex=2, paramNames [Highlight Color R, "Highlight  Color G", Highlight Color G, Highlight Color A]}
  NamedRerouteDeclaration(None=MaterialFunctionCall(R=@n20.Highlight Color R, G=@n20.Highlight Color R, B=@n20.Highlight Color R) {materialFunction=&AppendMany}.RGB) {name=Highlight Color, nodeColor=<rgba 0.2656 1 1 1>}
  NamedRerouteDeclaration(None=@n20.Highlight Color A) {name=Highlight Color A, nodeColor=<rgba 0.2656 1 1 1>}
  n21 = Multiply(A=SceneColor!(UVs).RGB)
  n22 = NamedRerouteUsage()
  n23 = NamedRerouteUsage()
  n24 = Add(A=NamedRerouteUsage(), B=NamedRerouteUsage())
  n25 = Desaturation(
   None=VectorParameter() {group=Color Values, parameterName=Base Color, defaultValue=<rgba 0.0296 0.0331 0.0497 1>}.RGB
   Fraction=ScalarParameter() {group=Color Values, parameterName=Color - Use Particle Color?}
  )
  n26 = Multiply(
   A=ParticleColor().RGB
   B=If(
    A=ScalarParameter() {parameterName=Color - Use Dynamic Colors?}
    "A > B"=LinearInterpolate(A=@n22, B=@n23, Alpha=@n24)
    "A == B"=@n25
    "A < B"=@n25
   )
  )
  n27 = NamedRerouteUsage()
  n28 = Constant()
  n29 = FeatureLevelSwitch(Default=@n27, ES2=@n28, ES3_1=@n28, SM4_REMOVED=@n28, SM5=@n27, SM6=@n27)
  NamedRerouteDeclaration(None=If(
    A=ScalarParameter() {parameterName=Color - Blend Base and Scene Color?}
    "A > B"=LinearInterpolate(A=@n21, B=@n26, Alpha=@n29)
    "A == B"=@n26
    "A < B"=@n26
   )) {name=Base_Color_Output, nodeColor=<rgba 0 1 0.2969 1>}
  NamedRerouteDeclaration(None=DynamicParameter() {parameterIndex=3, paramNames [Emissive Gain, Param2, Param3, Param4]}.Emissive Gain) {name=Emissive_Power, nodeColor=<rgba 1 0 0.7018 1>}
  n30 = SphereMask(
   A=NamedRerouteUsage()
   B=NamedRerouteUsage()
   Radius=ScalarParameter() {defaultValue=0.425, group=Center Fill, parameterName=Edge Highlight - Interior Radius}
   Hardness=ScalarParameter() {defaultValue=100, group=Edge Highlight, parameterName=Edge Highlight - Interior Radius Hardness}
  ) {attenuationRadius=0.4}
  n31 = NamedRerouteUsage()
  n32 = NamedRerouteUsage()
  n33 = Subtract(
   A=@n30
   B=SphereMask(
    A=Add(A=@n31, B=@n32)
    B=NamedRerouteUsage()
    Radius=ScalarParameter() {defaultValue=0.45, group=Edge Highlight, parameterName=Edge Highlight - Radius}
    Hardness=ScalarParameter() {defaultValue=0.7, group=Edge Highlight, parameterName=Edge Highlight - Hardness}
   ) {hardnessPercent=70}
  )
  NamedRerouteDeclaration(None=Multiply(A=Multiply(A=@n30, B=@n33) {constB=10}, B=NamedRerouteUsage())) {name=Edge_Highlight, nodeColor=<rgba 0.2307 1 0.5457 1>}
  n34 = ScalarParameter() {defaultValue=0.075, group=Highlight, parameterName=Center Highlight - Radius}
  n35 = Add(
   A=NamedRerouteUsage()
   B=Add(A=ScalarParameter() {defaultValue=-0.165, group=Highlight, parameterName=Center Highlight - Offset}, B=NamedRerouteUsage())
  )
  n36 = Multiply(
   A=SphereMask(A=@n35, B=NamedRerouteUsage(), Radius=Multiply(A=@n34) {constB=1.75}) {attenuationRadius=0.25, hardnessPercent=0}
   B=ScalarParameter() {defaultValue=0.2, group=Highlight, parameterName=Center Highlight -Fade Opacity}
  ) {constB=0.3}
  n37 = Multiply(
   A=SphereMask(A=@n35, B=NamedRerouteUsage(), Radius=ScalarParameter() {defaultValue=0.075, group=Highlight, parameterName=Center Highlight - Radius}) {attenuationRadius=0.4, hardnessPercent=0}
   B=ScalarParameter() {defaultValue=1.5, group=Opacity, parameterName=Center Highlight - Opacity}
  )
  NamedRerouteDeclaration(None=Multiply(A=Add(A=@n36, B=@n37), B=NamedRerouteUsage())) {name=Center_Highlight, nodeColor=<rgba 0.2307 1 0.5457 1>}
  n38 = ScalarParameter() {defaultValue=-0.165, group=Highlight, parameterName=Center Highlight - Offset}
  n39 = NamedRerouteUsage()
  n40 = Add(A=Divide(A=Add(A=@n38, B=@n39)) {constB=5}, B=NamedRerouteUsage())
  n41 = NamedRerouteUsage()
  n42 = Add(
   A=ScalarParameter() {defaultValue=0.425, group=Center Fill, parameterName=Center Fill - Radius}
   B=MaterialFunctionCall(Input=@n1, Input Low=Constant(), Input High=Constant() {r=2}, Target Low=Constant() {r=-0.1}, Target High=Constant() {r=0.3}) {materialFunction=&RemapValueRange}.Result
  )
  n43 = ScalarParameter() {defaultValue=65, group=Center Fill, parameterName=Center Fill - Hardness}
  n44 = ScalarParameter() {parameterName=Color - Blend Base and Scene Color?}
  n45 = ScalarParameter() {defaultValue=1, group=Center Fill, parameterName=Center Fill - Opacity}
  n46 = NamedRerouteUsage()
  NamedRerouteDeclaration(None=Multiply(
    A=SphereMask(A=@n40, B=@n41, Radius=@n42, Hardness=@n43) {attenuationRadius=0.4, hardnessPercent=65}
    B=If(A=@n44, "A > B"=@n45, "A == B"=@n46, "A < B"=@n46)
   ) {constB=0.3}) {name=Center Fill, nodeColor=<rgba 0.2307 1 0.5457 1>}
  n47 = ScalarParameter() {defaultValue=800, group=Opacity, parameterName=Opacity - DepthFade - Min}
  n48 = ScalarParameter() {defaultValue=100, group=Depth Fade, parameterName=Opacity - DepthFade - Max}
  n49 = NamedRerouteUsage()
  n50 = ScalarParameter() {defaultValue=1.5, group=Opacity, parameterName=Center Highlight - Opacity}
  n51 = NamedRerouteUsage()
  n52 = Multiply(A=NamedRerouteUsage(), B=ScalarParameter() {defaultValue=1, group=Opacity, parameterName=Edge Highlight - Opacity})
  n53 = Multiply(A=Add(A=Multiply(A=@n49, B=@n50), B=Add(A=@n51, B=@n52)), B=SphereMask(A=NamedRerouteUsage(), B=NamedRerouteUsage()) {attenuationRadius=0.4})
  n54 = ScalarParameter() {defaultValue=1, group=Opacity, parameterName=Global Opacity Multiplier}
  n55 = Multiply(A=Multiply(A=ParticleColor().A, B=Multiply(A=@n53, B=@n54)), B=NamedRerouteUsage())
  n56 = ShadingPathSwitch(Default=@n55, Deferred=DepthFade(Opacity=@n55, FadeDistance=LinearInterpolate(A=@n47, B=@n48, Alpha=@n55))) {desc="Defered is pc, Forward is switch, mobile is mobile"}
  n57 = FeatureLevelSwitch(Default=@n55, ES2=@n55, ES3_1=@n55, SM4_REMOVED=@n55, SM5=@n56, SM6=@n56)
  n58 = Constant() {r=1}
  NamedRerouteDeclaration(None=Multiply(A=@n55, B=StaticSwitchParameter(True=@n57, False=@n58) {group=Opacity, parameterName=UseDephtFade, sortPriority=10})) {name=Opacity Output, nodeColor=<rgba 0.2664 0.7758 1 1>}
  n59 = NamedRerouteUsage()
  n60 = NamedRerouteUsage()
  n61 = Multiply(A=NamedRerouteUsage(), B=ScalarParameter() {defaultValue=1, group=Opacity, parameterName=Edge Highlight - Opacity})
  n62 = Clamp(None=Add(A=Add(A=@n60, B=@n61), B=NamedRerouteUsage()), Max=NamedRerouteUsage()) {minDefault=0.025}
  n63 = VectorParameter() {parameterName=Shadow Color, defaultValue=<rgba 0.0199 0.0283 0.0677 1>}
  n64 = VectorParameter() {parameterName=Highlight Color, defaultValue=<rgba 0.3185 0.4397 1 1>}
  n65 = NamedRerouteUsage()
  n66 = Add(A=Add(A=@n65, B=NamedRerouteUsage()), B=NamedRerouteUsage())
  n67 = ScalarParameter() {defaultValue=1, group=Color Values, parameterName=Color - Shadow Strength}
  n68 = ParticleColor()
  n69 = ScalarParameter() {defaultValue=2, group=Color Values, parameterName=Color - Highlight Strength}
  n70 = Multiply(
   A=Desaturation(
    None=LinearInterpolate(A=@n63.RGB, B=@n64.RGB, Alpha=@n66)
    Fraction=ScalarParameter() {group=Color Values, parameterName=Color - Use Particle Color?}
   )
   B=LinearInterpolate(A=Multiply(A=@n67, B=@n68.RGB) {constB=0.75}, B=Multiply(A=@n68.RGB, B=@n69) {constB=3}, Alpha=@n65)
  )
  n71 = Multiply(
   A=If(A=ScalarParameter() {parameterName=Color - Use Dynamic Colors?}, "A > B"=Multiply(A=@n59, B=@n62), "A == B"=@n70, "A < B"=@n70)
   B=MaterialFunctionCall(
    Color_Morning=VectorParameter() {group=Color Unlit, parameterName=Color - Morning Multiplier, defaultValue=<rgba 1.15 0.8126 0.9177 0>}.RGB
    Color_Day=VectorParameter() {group=Color Unlit, parameterName=Color - Day Multiplier, defaultValue=<rgba 1.25 1.25 1.25 0>}.RGB
    Color_Evening=VectorParameter() {group=Color Unlit, parameterName=Color - Evening Multiplier, defaultValue=<rgba 1.15 0.7687 0.8251 0>}.RGB
    Color_Night=VectorParameter() {group=Color Unlit, parameterName=Color - Night Multiplier, defaultValue=<rgba 0.5417 0.7531 1 0>}.RGB
   ) {materialFunction=&Color_Time_of_Day}.Result
  )
  n72 = ScalarParameter() {defaultValue=1.25, parameterName=Color - Emissive Gain}
  NamedRerouteDeclaration(None=Power(Base=Multiply(A=@n71, B=@n72) {constB=1.5}, Exp=NamedRerouteUsage())) {name=Color_Output, nodeColor=<rgba 1 0.1562 0.8606 1>}
  NamedRerouteDeclaration(None=@n2.Dissolve) {name=Dissolve, nodeColor=<rgba 1 0 0.0132 1>}
  NamedRerouteDeclaration(None=@n2.Wiggle) {name=Wiggle, nodeColor=<rgba 1 0 0.0132 1>}
  NamedRerouteDeclaration(None=@n2.Refraction) {name=Refraction Amount, nodeColor=<rgba 1 0 0.0132 1>}
  NamedRerouteUsage()
  NamedRerouteUsage()
  NamedRerouteUsage()
  TextureSampleParameter2D() {
   group=reflection
   parameterName=Normal
   samplerType=SAMPLERTYPE_Normal
   texture=/Game/Gadgets/Assets/Teleporter/Textures/T_Spherical_Normal_02
  }
 ]
}
